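; BP_Commerce_Cheese_Group_H is a cheese prop group for a market, varying the color of its cheese meshes.
; Its construction script calls the parent version, then loops over three components: two old cheese wheels and one round cheese.
; Each is passed to SetColoronSecondaryMesh with a random pick from the TintColor_Cheese array.
; The BeginPlay, overlap and Tick events are empty, so the script adds no runtime logic.
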(event EventBeginPlay)

(event Collision|EventActorBeginOverlap
  (params (OtherActor :type Actor%20Object%20Reference)))

(event EventTick
  (params (DeltaSeconds :type Float%20%28single-precision%29)))

(fn ConstructionScript
  (|CallParentFunction)
  (for _array_element (Utilities|Array|MakeArray (Variables|Default|GetSM_Commerce_Old_Cheese_Wheel_C1) (Variables|Default|GetSM_Commerce_Round_Cheese_A) (Variables|Default|GetSM_Commerce_Old_Cheese_Wheel_C))
    (Default|SetColoronSecondaryMesh (Utilities|Array|RandomArrayItem (Variables|Default|GetTintColor_Cheese)) _array_element)))
