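; BP_Commerce_Cheese_Group_J is a market prop group of cheeses, jam jars, summer sausage and bread with randomized tinting.
; After the parent construction script, it loops over the cheese, jar and sausage meshes, setting each secondary-mesh color randomly from TintColor_Cheese.
; A loop nested inside that one tints two baguettes and a sourdough from TintColor_Bread, so bread is re-randomized on every outer pass.
; BeginPlay, overlap and Tick events are empty.
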(event EventBeginPlay)

(event Collision|EventActorBeginOverlap
  (params (OtherActor :type Actor%20Object%20Reference)))

(event EventTick
  (params (DeltaSeconds :type Float%20%28single-precision%29)))

(fn ConstructionScript
  (|CallParentFunction)
  (for _array_element (Utilities|Array|MakeArray (Variables|Default|GetSM_Commerce_Rectangular_Cheese_A1) (Variables|Default|GetSM_Commerce_Jam_Jars_A) (Variables|Default|GetSM_Commerce_Jam_Jars_A1) (Variables|Default|GetSM_Commerce_Jam_Jars_A3) (Variables|Default|GetSM_Commerce_Jam_Jars_A2) (Variables|Default|GetSM_Commerce_Summer_Sausage_A) (Variables|Default|GetSM_Commerce_Rectangular_Cheese_A) (Variables|Default|GetSM_Commerce_Round_Cheese_A) (Variables|Default|GetSM_Commerce_Square_Cheese_A) (Variables|Default|GetSM_Commerce_Old_Cheese_Wheel_C) (Variables|Default|GetSM_Commerce_Old_Cheese_Wheel_B))
    (Default|SetColoronSecondaryMesh (Utilities|Array|RandomArrayItem (Variables|Components|GetTintColor_Cheese)) _array_element)
    (for _array_element_2 (Utilities|Array|MakeArray (Variables|Default|GetSM_Commerce_Baguette_A) (Variables|Default|GetSM_Commerce_Baguette_A1) (Variables|Default|GetSM_Commerce_Sourdough_A))
      (Default|SetColoronSecondaryMesh (Utilities|Array|RandomArrayItem (Variables|Components|GetTintColor_Bread)) _array_element_2))))
